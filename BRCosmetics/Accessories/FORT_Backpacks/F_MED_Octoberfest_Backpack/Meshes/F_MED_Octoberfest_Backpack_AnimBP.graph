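# Animation Blueprint for the Octoberfest backpack cosmetic, adding procedural secondary motion on top of its input pose.
# Three chained Anim Dynamics nodes simulate the dyn_bread_child bones with angular limits and sim-space gravity.
# Their result feeds a RigidBody node with amplified, clamped component acceleration and its own gravity override before returning to local space.
# The event graph has no logic and there are no variables.
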AnimBlueprint /BRCosmetics/Accessories/FORT_Backpacks/F_MED_Octoberfest_Backpack/Meshes/F_MED_Octoberfest_Backpack_AnimBP
target_skeleton=&/F_MED_Octoberfest_Backpack_Skeleton
graphs [
 AnimGraph {nodes [
   AnimGraphNode_SubInput_0 = Input Pose()
   n1 = Anim Dynamics - Bone: dyn_bread_child_01(ComponentPose=Local To Component(LocalPose=@AnimGraphNode_SubInput_0), Alpha=1, ExternalForce="0,0,0") {
    angularDampingOverride=10
    angularSpringConstant=100
    bGravityOverrideInSimSpace=true
    bUseGravityOverride=true
    alphaBoolBlend {blendOption=Linear}
    boundBone {boneName=dyn_bread_child_01}
    componentLinearAccScale=<1 1 1>
    componentLinearVelScale=<1 1 1>
    gravityOverride {z=-980}
    physicsBodyDefinitions [{
      sphereCollisionRadius=0
      boundBone {boneName=dyn_bread_child_01}
      boxExtents=<20 20 10>
      constraintSetup {twistAxis=AxisY, angularLimitsMax=<0.5 2 0>, angularLimitsMin=<-0.5 -2 0>, angularTarget=<0 0 0>}
      localJointOffset=<-1 0 -1.2>
     }]
   }
   n2 = Anim Dynamics - Bone: dyn_bread_child_02(ComponentPose=@n1, Alpha=1, ExternalForce="0,0,0") {
    angularDampingOverride=10
    angularSpringConstant=100
    bGravityOverrideInSimSpace=true
    bUseGravityOverride=true
    alphaBoolBlend {blendOption=Linear}
    boundBone {boneName=dyn_bread_child_02}
    componentLinearAccScale=<1 1 1>
    componentLinearVelScale=<1 1 1>
    gravityOverride {z=-980}
    physicsBodyDefinitions [{
      sphereCollisionRadius=0
      boundBone {boneName=dyn_bread_child_02}
      boxExtents=<20 20 10>
      constraintSetup {twistAxis=AxisY, angularLimitsMax=<0.5 2 0>, angularLimitsMin=<-0.5 -2 0>, angularTarget=<0 0 0>}
      localJointOffset=<0 -1 -1>
     }]
   }
   n3 = Anim Dynamics - Bone: dyn_bread_child_03(ComponentPose=@n2, Alpha=1, ExternalForce="0,0,0") {
    angularDampingOverride=10
    angularSpringConstant=100
    bGravityOverrideInSimSpace=true
    bUseGravityOverride=true
    alphaBoolBlend {blendOption=Linear}
    boundBone {boneName=dyn_bread_child_03}
    componentLinearAccScale=<1 1 1>
    componentLinearVelScale=<1 1 1>
    gravityOverride {z=-980}
    physicsBodyDefinitions [{
      sphereCollisionRadius=0
      boundBone {boneName=dyn_bread_child_03}
      boxExtents=<20 20 10>
      constraintSetup {twistAxis=AxisY, angularLimitsMax=<0.5 2 0>, angularLimitsMin=<-0.5 -2 0>, angularTarget=<0 0 0>}
      localJointOffset=<-1 -1 -1.1>
     }]
   }
   n4 = RigidBody(ComponentPose=@n3) {
    bOverrideWorldGravity=true
    alphaBoolBlend {blendOption=Linear}
    componentAppliedLinearAccClamp=<1000 1000 1000>
    componentLinearAccScale=<20 20 20>
    componentLinearVelScale=<1 1 1>
    overrideWorldGravity {z=-980}
   }
   Output Pose(Result=Component To Local(ComponentPose=@n4)) {layerGroup=None}
  ]}
 EventGraph {nodes [Event BlueprintUpdateAnimation()]}
]
variables []
