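# Physics asset for the Octoberfest backpack cosmetic, setting up a hinged lid for physics on the backpack mesh.
# A small zero-length capsule on backpack_root and a larger sphere on dyn_lid provide the two bodies.
# One constraint joins the lid to the root, locking twist and the first swing axis so the lid rotates only about the second swing axis.
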PhysicsAsset /BRCosmetics/Accessories/FORT_Backpacks/F_MED_Octoberfest_Backpack/Meshes/F_MED_Octoberfest_Backpack_Physics
bodies {
 backpack_root [{shapeType=Capsule, length=0, radius=1.5085, center=<0.1006 0.0872 0.126>, rotation=<rot 90 0 0>}]
 dyn_lid [{shapeType=Sphere, radius=3.75, center=<-0.1383 0 0.1827>}]
}
constraints [{
  bone1Name=dyn_lid
  bone2Name=backpack_root
  swing1LimitDegrees=45
  swing1Motion=Locked
  swing2LimitDegrees=30
  twistLimitDegrees=45
  twistMotion=Locked
 }]
